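# Physics asset for the CashmereScarf two-handed melee pickaxe skeletal mesh in the BR cosmetics weapons set.
# It holds a single body on the Melee_TwoHanded_Handle bone, made of one tilted capsule offset along the handle.
# No constraints are defined, so the asset provides one simple collision shape rather than an articulated simulation.
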PhysicsAsset /BRCosmetics/Weapons/FORT_Melee/Pickaxe_CashmereScarf/Meshes/CashmereScarf_Axe_PhysicsAsset
bodies {Melee_TwoHanded_Handle [{shapeType=Capsule, length=24.6246, radius=3.2671, center=<-0.4854 -1.1074 14.7269>, rotation=<rot 0.8247 -90 0>}]}
constraints []
